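# Post-process-style AnimBlueprint for the SwiftKestrel head skeleton, adapting its deform bones to player actions.
# Event graph turns sprint, emote, weapon type, aiming, ziplining and skydiving into float alphas feeding a Control Rig.
# Chest and upper-arm helper bones are then layered in from the body mesh's pose, cached when customization loads.
# Frontend, crouch-tail and melee values are computed but unused; the parachute branch has no execution input.
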AnimBlueprint /BRCosmetics/Characters/Player/Female/Medium/Heads/F_MED_SwiftKestrel_Head/Meshes/CustomSkeleton/Fortnite_Base_Head_SwiftKestrel_AnimB
target_skeleton=&/Fortnite_Base_Head_SwiftKestrel_Skeleton
graphs [
 AnimGraph {nodes [
   Layered blend per bone() {layerSetup [{branchFilters [
       {boneName=dfrm_clavicle_l}
       {boneName=dfrm_clavicle_r}
       {boneName=dfrm_upperarm_l}
       {boneName=dfrm_upperarm_r}
       {boneName=dfrm_upperarm_twist_01_l}
       {boneName=dfrm_upperarm_twist_01_r}
       {boneName=dyn_wing_upper_l_MSTR}
       {boneName=dyn_wing_upper_r_MSTR}
       {boneName=dyn_wing_upper_r_twist_MSTR}
       {boneName=dyn_wing_upper_l_twist_MSTR}
      ]}]}
   n1 = Control Rig!(
    Source=Input Pose()
    Alpha
    sprint:double=Get sprint() -> (sprint:double)
    emote:double=Get emote() -> (emote:double)
    pistol:double=Get pistol() -> (pistol:double)
    skydivezip_alpha:double=Get zip_skydive_alpha() -> (zip_skydive_alpha:double)
    parachute:double=Get parachute() -> (parachute:double)
    aim_alpha:double=Get aim_alpha() -> (aim_alpha:double)
    shotgun:double=Get shotgun() -> (shotgun:double)
    idle:bool=Get bRigidBodyEnabled()
   ) -> (Pose) {controlRigAssetReference {blueprintRigClass=/BRCosmetics/Characters/Player/Female/Medium/Heads/F_MED_SwiftKestrel_Head/Meshes/F_MED_SwiftKestrel_Head_CtrlRig.F_MED_SwiftKestrel_Head_CtrlRig_C}}
   n2 = Layered blend per bone(BasePose=@n1, BlendPoses_0=Copy Pose From Mesh(SourceMeshComponent=Get Body_Mesh())) {layerSetup [{branchFilters [{boneName=dfrm_chest_01_l}, {boneName=dfrm_chest_01_r}, {boneName=dfrmbase_upperarm_up_l}, {boneName=dfrmbase_upperarm_up_r}]}]}
   Output Pose(Result=@n2)
  ]}
 EventGraph {nodes [
   Event OnCharacterCustomizationLoaded()
   > n1 = Cast To PlayerPawn_Athena(Object=TryGetPawnOwner())
   > Set Body_Mesh(Body_Mesh=GetSkeletalMeshForPartType(self=@n1))
   Event OnBeginTargeting() {is=title K2Node_Event}
   > Set aim_alpha(aim_alpha:double=1) -> (Output_Get:double)
   Event OnEndTargeting() {is=title K2Node_Event}
   > Set aim_alpha(aim_alpha=0)
   n2 = Event BlueprintProcessAnimInputEvent()
   > n3 = Branch(Condition=IsNamed(self=@n2.AnimInputEvent, InEventName=aie.Player.Zipline.Entered))
   > Set ziplining_alpha(ziplining_alpha:bool=1) -> (Output_Get:bool)
   Branch(execute=@n3.else, Condition=IsNamed(self=@n2.AnimInputEvent, InEventName=aie.Player.Zipline.Exited))
   > Set ziplining_alpha(ziplining_alpha=false)
   Event BlueprintUpdateAnimation()
   > n4 = Sequence!(execute) -> (then_0, then_1, then_2:Exec, then_3:Exec, then_4:Exec, then_5:Exec)
   n5 = Get bIsSprinting()
   n6 = Get bIsPlayingEmote()
   > Set sprint(sprint:double="To Float (Boolean)"(InBool=OR Boolean(A=@n5, B=@n6))) -> (Output_Get:double)
   > Set emote(emote:double="To Float (Boolean)"(InBool=Get bIsPlayingEmote())) -> (Output_Get:double)
   > Set isSkyDivingFloat(isSkyDivingFloat:double="To Float (Boolean)"(InBool=Get bIsSkyDiving())) -> (Output_Get:double)
   > n7 = Cast To FortPlayerAnimInstance(Object=GetAnimInstance(self=GetOwnerSkeletalMeshComponent()))
   > Set pistol(pistol:double=-0.6) -> (Output_Get:double)
   > Set shotgun(shotgun:double=-0.9) -> (Output_Get:double)
   > Set melee_alpha(melee_alpha:double=-0.01) -> (Output_Get:double)
   > n8 = Switch on EFortWeaponCoreAnimation!(execute, Selection:EFortWeaponCoreAnimation Enum=Get WeaponCoreAnim(self=@n7)) -> (
    Melee:Exec
    Pistol:Exec
    Shotgun:Exec
    PaperBlueprint:Exec
    Rifle:Exec
    MeleeOneHand:Exec
    MachinePistol:Exec
    RocketLauncher:Exec
    GrenadeLauncher:Exec
    GoingCommando:Exec
    AssaultRifle:Exec
    TacticalShotgun:Exec
    SniperRifle:Exec
    TrapPlacement:Exec
    ShoulderLauncher:Exec
    AbilityDecoTool:Exec
    Crossbow:Exec
    C4:Exec
    RemoteControl:Exec
    DualWield:Exec
    AR_BullPup:Exec
    AR_ForwardGrip:Exec
    MedPackPaddles:Exec
    SMG_P90:Exec
    AR_DrumGun:Exec
    Consumable_Small:Exec
    Consumable_Large:Exec
    Balloon:Exec
    MountedTurret:Exec
    CreativeTool:Exec
    ExplosiveBow:Exec
    AshtonIndigo:Exec
    AshtonChicago:Exec
    MeleeDualWield:Exec
    Unarmed:Exec
   )
   Set pistol(execute=@n8.MachinePistol, pistol=-0.1)
   Set melee_alpha(execute=@n8.Melee, melee_alpha=-0.3)
   Set pistol(execute=@n8.Pistol, pistol=-0.1)
   Set shotgun(execute=@n8.Shotgun, shotgun=-0.2)
   Set shotgun(execute=@n8.TacticalShotgun, shotgun=-0.2)
   n9 = Get bRigidBodyEnabled()
   n10 = Branch(execute=@n4.then_2, Condition=AND Boolean(A=NOT Boolean(A=@n9), B=Get bIsCrouching()))
   > Set tailcrouch_alpha(tailcrouch_alpha:double=1) -> (Output_Get:double)
   Set tailcrouch_alpha(execute=@n10.else, tailcrouch_alpha=0)
   n11 = Get bIsSkyDiving()
   n12 = Get bIsPlayingEmote()
   n13 = Branch(execute=@n4.then_1, Condition=AND Boolean!(A=Get bIsInFrontEnd(), B=NOT Boolean(A=@n12), C:bool=NOT Boolean(A=@n11)) -> (ReturnValue))
   > Set frontend(frontend:double=1) -> (Output_Get:double)
   Set frontend(execute=@n13.else, frontend=0)
   n14 = Branch(execute=@n4.then_3, Condition=OR Boolean(A=Get ziplining_alpha() -> (ziplining_alpha:bool), B=Get bIsSkyDiving()))
   > Set zip_skydive_alpha(zip_skydive_alpha:double=1) -> (Output_Get:double)
   Set zip_skydive_alpha(execute=@n14.else, zip_skydive_alpha=0)
   n15 = Branch(Condition=Get bIsParachuteOpened())
   > Set parachute(parachute:double=1) -> (Output_Get:double)
   Set parachute(execute=@n15.else, parachute=0)
   TryGetPawnOwner()
  ]}
]
variables [
 aim_alpha {type=number}
 Body_Mesh {type=SkeletalMeshComponent}
 emote {type=number}
 frontend {type=number}
 isSkyDivingFloat {type=number}
 melee_alpha {type=number}
 parachute {type=number}
 pistol {type=number}
 shotgun {type=number}
 sprint {type=number}
 tailcrouch_alpha {type=number}
 zip_skydive_alpha {type=number}
 ziplining_alpha {type=boolean}
]
